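; Looping gameplay cue for Athena "Alpaca" that plays a 2D ambient sound for the locally viewed pawn while the cue is active.
; On application it checks that the target is a locally viewed FortPawn, stores it and calls EnableAmbientAudio.
; That function spawns the ambient asset and fades it in from a random start time to vary the loop.
; On removal, enabled audio fades out through the stored component.
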(event EventBeginPlay)

(event Collision|EventActorBeginOverlap
  (params (OtherActor :type Actor%20Object%20Reference)))

(event EventTick
  (params (DeltaSeconds :type Float%20%28single-precision%29)))

(event EventOnApplication(EitherVFX-preferthisone)
  (params (MyTarget :type Actor%20Object%20Reference) (Parameters :type Gameplay%20Cue%20Parameters%20Structure%20%28by%20ref%29) (ParticleComponents :type Array%20of%20FXSystem%20Component%20Object%20References) (AudioComponents :type Array%20of%20Audio%20Component%20Object%20References) (BurstCameraShakeInstance :type Camera%20Shake%20Base%20Object%20Reference) (BurstDecalInstance :type Decal%20Actor%20Object%20Reference))
  (bind _asfort_pawn (Utilities|Casting|CastToFortPawn MyTarget))
  (if (Pawn|IsLocallyViewed _asfort_pawn)
    (Variables|Default|SetPawnTarget _asfort_pawn)
    (CallFunction|EnableAmbientAudio true)))

(event EventOnRemoval(EitherVFX-preferthisversion)
  (params (MyTarget :type Actor%20Object%20Reference) (Parameters :type Gameplay%20Cue%20Parameters%20Structure%20%28by%20ref%29) (ParticleComponents :type Array%20of%20FXSystem%20Component%20Object%20References) (AudioComponents :type Array%20of%20Audio%20Component%20Object%20References) (BurstCameraShakeInstance :type Camera%20Shake%20Base%20Object%20Reference) (BurstDecalInstance :type Decal%20Actor%20Object%20Reference))
  (if (Variables|Default|GetAmbientEnabled)
    (CallFunction|EnableAmbientAudio false)))

(fn EnableAmbientAudio
  (params (Enabled :type Boolean))
  (bind _output_get (Variables|Default|SetAmbientEnabled Enabled))
  (if _output_get
    (bind _returnvalue (Audio|SpawnSound2D 0 (Variables|Default|GetAmbientAsset)))
    (bind _output_get_2 (Variables|Default|SetAmbientAudioComp _returnvalue))
    (Audio|Components|Audio|FadeIn _output_get_2 1.0 1.0 (Math|Random|RandomFloatinRange 0.0 9e+01) "Linear" :!class /Script/Engine.AudioComponent)
    (else
      (bind _ambientaudiocomp (Variables|Default|GetAmbientAudioComp)
        (:then
          (Audio|Components|Audio|FadeOut _ambientaudiocomp 1.0 0.0 "Linear" :!class /Script/Engine.AudioComponent))
        (:else)))))

(fn ConstructionScript)
