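; Looping gameplay cue for the BadNews weapon charge-up, driving muzzle charge particles and sound.
; On loop start it caches the spawned particle and audio components, then feeds MinChargeTime from the effect causer pawn's current weapon stats into the particle's ChargeUpLength parameter.
; On removal it fades out the charge sound and schedules its stop after the fade duration.
; Debug PrintString calls remain in both start and removal paths.
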(event EventBeginPlay)

(event Collision|EventActorBeginOverlap
  (params (OtherActor :type Actor%20Object%20Reference)))

(event EventTick
  (params (DeltaSeconds :type Float%20%28single-precision%29)))

(event EventOnLoopingStart(EitherVFX-preferthisversion)
  (params (MyTarget :type Actor%20Object%20Reference) (Parameters :type Gameplay%20Cue%20Parameters%20Structure%20%28by%20ref%29) (ParticleComponents :type Array%20of%20FXSystem%20Component%20Object%20References) (AudioComponents :type Array%20of%20Audio%20Component%20Object%20References))
  (bind (_normalizedmagnitude _rawmagnitude _effectcontext _matchedtagname _originaltag _aggregatedsourcetags _aggregatedtargettags _location _normal _instigator _effectcauser) (Ability|GameplayCue|BreakGameplayCueParameters Parameters))
  (|CallParentFunction MyTarget Parameters ParticleComponents AudioComponents)
  (bind _result (ArrayMacros|SafeArrayGet ParticleComponents)
    (:%20%20
      (Variables|Default|SetMuzzleChargeFX _result)
      (bind _result_2 (ArrayMacros|SafeArrayGet AudioComponents)
        (:%20%20
          (Variables|Default|SetMuzzleChargeSound _result_2)
          ;[1 1 1] Get Min Charge Time
          (bind _asfort_pawn (Utilities|Casting|CastToFortPawn _effectcauser))
          (bind _outrow (Fortnite|GetRangedWeaponStatsRow (Data|GetWeaponDataWeaponStatHandle (Variables|/Script/FortniteGame.FortPawn|GetCurrentWeapon _asfort_pawn))))
          (Development|PrintString 0 "Hello" true true "(R=0.000000,G=0.660000,B=1.000000,A=1.000000)" (.MinChargeTime (Utilities|Struct|BreakFortRangedWeaponStats _outrow)))
          (Effects|Components|ParticleSystem|SetFloatParameter (Variables|Default|GetMuzzleChargeFX) "ChargeUpLength" (.MinChargeTime (Utilities|Struct|BreakFortRangedWeaponStats _outrow)))
          (Development|PrintString 0 (Utilities|String|ToString(Float) (.MinChargeTime (Utilities|Struct|BreakFortRangedWeaponStats _outrow))) true true "(R=1.000000,G=0.046466,B=0.000000,A=1.000000)" 5.0))))))

(event EventOnRemoval(EitherVFX-preferthisversion)
  (params (MyTarget :type Actor%20Object%20Reference) (Parameters :type Gameplay%20Cue%20Parameters%20Structure%20%28by%20ref%29) (ParticleComponents :type Array%20of%20FXSystem%20Component%20Object%20References) (AudioComponents :type Array%20of%20Audio%20Component%20Object%20References) (BurstCameraShakeInstance :type Camera%20Shake%20Base%20Object%20Reference) (BurstDecalInstance :type Decal%20Actor%20Object%20Reference))
  (|CallParentFunction MyTarget Parameters ParticleComponents AudioComponents BurstCameraShakeInstance BurstDecalInstance)
  (Development|PrintString 0 "Removed" true true "(R=1.000000,G=0.000000,B=0.127922,A=1.000000)" 5.0)
  (Audio|Components|Audio|FadeOut (Variables|Default|GetMuzzleChargeSound) (Variables|Default|GetMuzzleChargeSound_FadeOutDuration) 0.0 "Linear" :!class /Script/Engine.AudioComponent)
  (Audio|Components|Audio|StopDelayed (Variables|Default|GetMuzzleChargeSound) (Variables|Default|GetMuzzleChargeSound_FadeOutDuration)))

(fn ConstructionScript)
