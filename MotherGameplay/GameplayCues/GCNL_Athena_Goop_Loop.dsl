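; Looping gameplay cue that gives the locally controlled pawn a screen and audio effect while affected by Goop.
; On start it pushes an interior sound mix and plays a timeline driving post-process blend weight and two global material parameters.
; Removal pops the mix and reverses the timeline, or clears instantly if dead.
; A health-changed binding also clears at zero health, because dead pawns stop reporting as locally controlled.
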(event Custom|HealthChanged
  ;[1 1 1] Use Health changed to get "predeath" state of health <=0, since dead actors don't return as locally controlled. Used to ensure dead actors lose the PP, and that their killers don't
  (if (<= (FortPawn|GetHealth (Variables|Default|GetMy%20Target)) 0.0)
    (CallFunction|RemovePostProcessInstantly)))

(event EventOnLoopingStart(Niagara)
  (params (MyTarget :type Actor%20Object%20Reference) (Parameters :type Gameplay%20Cue%20Parameters%20Structure%20%28by%20ref%29) (ParticleComponents :type Array%20of%20Niagara%20Particle%20System%20Component%20Object%20References) (AudioComponents :type Array%20of%20Audio%20Component%20Object%20References))
  (bind _asfort_pawn (Utilities|Casting|CastToFortPawn MyTarget))
  (bind _output_get (Variables|Default|SetMy%20Target _asfort_pawn))
  (if (Pawn|IsLocallyControlled (Variables|Default|GetMy%20Target))
    (Health|BindEventtoOnPawnHealthChanged _output_get (delegate Custom|HealthChanged))
    (CallFunction|SetInteriorAmbientEnabled true)
    (Utilities|FlowControl|Sequence
      (:then_0
        (bind _result (ArrayMacros|SafeArrayGet AudioComponents)
          (:%20%20
            (Variables|Audio|SetLoopingAudioComponent _result))))
      (:then_1
        (Variables|/Script/Engine.PostProcessComponent|SetbEnabled true (Variables|Default|GetGoopPostProcess))
        (bind (_direction _lerp _distortion) (Timeline|PP_Timeline Play)
          (:Update
            (Variables|/Script/Engine.PostProcessComponent|SetBlendWeight _lerp (Variables|Default|GetGoopPostProcess))
            (Rendering|Material|SetScalarParameterValue 0 "/Game/Packages/Fortress_SharedMaterials/GlobalMaterialParameters/FortniteMaterialParameters.FortniteMaterialParameters" "GoopPP" _lerp :!class /Script/Engine.KismetMaterialLibrary)
            (Rendering|Material|SetScalarParameterValue 0 "/Game/Packages/Fortress_SharedMaterials/GlobalMaterialParameters/FortniteMaterialParameters.FortniteMaterialParameters" "GoopPP_Distortion" _distortion :!class /Script/Engine.KismetMaterialLibrary))
          (:Finished
            (switch Utilities|FlowControl|Switch|SwitchonETimelineDirection _direction
              (:Backward
                (Variables|/Script/Engine.PostProcessComponent|SetbEnabled false (Variables|Default|GetGoopPostProcess))))))))))

(event EventOnRemoval(Niagara)
  (params (MyTarget :type Actor%20Object%20Reference) (Parameters :type Gameplay%20Cue%20Parameters%20Structure%20%28by%20ref%29) (ParticleComponents :type Array%20of%20Niagara%20Particle%20System%20Component%20Object%20References) (AudioComponents :type Array%20of%20Audio%20Component%20Object%20References) (BurstCameraShakeInstance :type Camera%20Shake%20Base%20Object%20Reference) (BurstDecalInstance :type Decal%20Actor%20Object%20Reference))
  (CallFunction|SetInteriorAmbientEnabled false)
  (if (Pawn|IsLocallyControlled (Variables|Default|GetMy%20Target))
    (if (not (FortPawn|IsDead (Variables|Default|GetMy%20Target)))
      (Timeline|PP_Timeline Reverse)
      (else
        ;[1 1 1] Clean up instantly if dead because this GCNL won't last through the timeline
        (Variables|/Script/Engine.PostProcessComponent|SetBlendWeight 0.0 (Variables|Default|GetGoopPostProcess))
        (Rendering|Material|SetScalarParameterValue 0 "/Game/Packages/Fortress_SharedMaterials/GlobalMaterialParameters/FortniteMaterialParameters.FortniteMaterialParameters" "GoopPP" :!class /Script/Engine.KismetMaterialLibrary)
        (Rendering|Material|SetScalarParameterValue 0 "/Game/Packages/Fortress_SharedMaterials/GlobalMaterialParameters/FortniteMaterialParameters.FortniteMaterialParameters" "GoopPP_Distortion" :!class /Script/Engine.KismetMaterialLibrary)
        (CallFunction|SetInteriorAmbientEnabled false)
        (Variables|/Script/Engine.PostProcessComponent|SetbEnabled false (Variables|Default|GetGoopPostProcess))))
    (elif (<= (Utilities|Array|Length (Collision|GetOverlappingActors (Variables|Default|GetMy%20Target) (Variables|Default|GetBGA_Goop))) 0)
      (CallFunction|RemovePostProcessInstantly))))

(event Custom|RemovePostProcessInstantly
  ;[1 1 1] Clean up instantly if dead because this GCNL won't last through the timeline
  (Variables|/Script/Engine.PostProcessComponent|SetBlendWeight 0.0 (Variables|Default|GetGoopPostProcess))
  (Rendering|Material|SetScalarParameterValue 0 "/Game/Packages/Fortress_SharedMaterials/GlobalMaterialParameters/FortniteMaterialParameters.FortniteMaterialParameters" "GoopPP" :!class /Script/Engine.KismetMaterialLibrary)
  (Rendering|Material|SetScalarParameterValue 0 "/Game/Packages/Fortress_SharedMaterials/GlobalMaterialParameters/FortniteMaterialParameters.FortniteMaterialParameters" "GoopPP_Distortion" :!class /Script/Engine.KismetMaterialLibrary)
  (CallFunction|SetInteriorAmbientEnabled false)
  (Variables|/Script/Engine.PostProcessComponent|SetbEnabled false (Variables|Default|GetGoopPostProcess)))

(fn SetInteriorAmbientEnabled
  (params (Enabled :type Boolean))
  (bind _my_target (Variables|Default|GetMy%20Target))
  (if (or (Pawn|IsLocallyControlled (Variables|Default|GetMy%20Target)) (Utilities|Equal(Object) (Variables|Default|GetMy%20Target) (Variables|Default|GetSoundMixPawn)))
    (if Enabled
      (Audio|PushSoundMixModifier _my_target "/MotherGameplay/Items/Goop/Audio/Goop_Inside_SoundMix.Goop_Inside_SoundMix")
      (Variables|Default|SetSoundMixPawn _my_target)
      (else
        (Audio|PopSoundMixModifier _my_target "/MotherGameplay/Items/Goop/Audio/Goop_Inside_SoundMix.Goop_Inside_SoundMix")))))

(fn ConstructionScript)
