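; Looping gameplay cue for the Llama Roaster beam fire, driving its Cascade beam endpoint and flyby audio.
; Tick targets the local player's aim point directly; remote pawns use replicated aim, with interpolation from a saved point.
; Loop start fades in flyby audio for remote pawns only.
; Update Audio Location, called from Tick's local branch, projects the nearest listener onto the forward axis to place that sound, stopping it for listeners behind.
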(event EventTick
  (params (DeltaSeconds :type Float%20%28single-precision%29))
  (bind _returnvalue (Transformation|GetActorLocation))
  ;[1 1 1] Interpolate the fx rotation to match control rotation
  (Utilities|IsValid (stale EventOnLoopingStart(Cascade) MyTarget)
    (:Is%20Valid
      (Utilities|IsValid (Utilities|Array|Get(acopy) (stale EventOnLoopingStart(Cascade) ParticleComponents))
        (:Is%20Valid
          (if (Pawn|IsLocallyControlled (Utilities|Casting|CastToFortPawn (stale EventOnLoopingStart(Cascade) MyTarget)))
            ;;[1 1 1] If this is the local client. don't used the replicated hit location, just grab it directly.
            (Transformation|SetWorldRotation (Utilities|Array|Get(acopy) (stale EventOnLoopingStart(Cascade) ParticleComponents)) (Pawn|GetBaseAimRotation (Utilities|Casting|CastToFortPawn (stale EventOnLoopingStart(Cascade) MyTarget))) false false)
            (CallFunction|UpdateAudioLocation)
            (bind _returnvalue_2 (Fortnite|GetPlayerAimPoint (Fortnite|GetFortPlayerControllerfromActor (Utilities|Casting|CastToFortPawn (stale EventOnLoopingStart(Cascade) MyTarget))) 1.5e+04 (Utilities|Array|MakeArray (Utilities|Casting|CastToFortPawn (stale EventOnLoopingStart(Cascade) MyTarget)))))
            (Effects|Components|ParticleSystem|SetBeamTargetPoint (Utilities|Array|Get(acopy) (stale EventOnLoopingStart(Cascade) ParticleComponents)) 0 _returnvalue_2 0)
            (else
              ;[1 1 1] Interpolate the GC's rotation towards the replicated control rotation
              (bind _asplayer_pawn_athena (Utilities|Casting|CastToPlayerPawn_Athena (stale EventOnLoopingStart(Cascade) MyTarget)))
              (if (Math|Vector|NotEqual(Vector) (Variables|Default|GetSavedPoint) (+ _returnvalue (* (Math|Vector|GetForwardVector (Pawn|GetBaseAimRotation _asplayer_pawn_athena)) 1.5e+04)) 0.0001)
                (Utilities|DoOnceThenDo false
                  (:DoOnce
                    (Variables|Default|SetSavedPoint (+ _returnvalue (* (Math|Vector|GetForwardVector (Pawn|GetBaseAimRotation _asplayer_pawn_athena)) 1.5e+04))))
                  (:ThenDo
                    (Effects|Components|ParticleSystem|SetBeamTargetPoint (Utilities|Array|Get(acopy) (stale EventOnLoopingStart(Cascade) ParticleComponents)) 0 (Math|Interpolation|VInterpTo (Variables|Default|GetSavedPoint) (+ _returnvalue (* (Math|Vector|GetForwardVector (Pawn|GetBaseAimRotation _asplayer_pawn_athena)) 1.5e+04)) DeltaSeconds 3e+01) 0)))))))))))

(event EventOnLoopingStart(Cascade)
  (params (MyTarget :type Actor%20Object%20Reference) (Parameters :type Gameplay%20Cue%20Parameters%20Structure%20%28by%20ref%29) (ParticleComponents :type Array%20of%20Cascade%20Particle%20System%20Component%20Object%20References) (AudioComponents :type Array%20of%20Audio%20Component%20Object%20References))
  (bind _returnvalue (Transformation|GetActorLocation))
  ;[1 1 1] Interpolate the fx rotation to match control rotation
  (Variables|Default|SetPawn MyTarget)
  (Utilities|FlowControl|Sequence
    (:then_0
      (Utilities|FlowControl|Sequence
        (:then_0
          _)
        (:then_1
          ;[1 1 1] Interpolate the GC's rotation towards the replicated control rotation
          (bind _asplayer_pawn_athena (Utilities|Casting|CastToPlayerPawn_Athena MyTarget))
          (if (Math|Vector|NotEqual(Vector) (Variables|Default|GetSavedPoint) (+ _returnvalue (* (Math|Vector|GetForwardVector (Pawn|GetBaseAimRotation _asplayer_pawn_athena)) 1.5e+04)) 0.0001)
            _)))
      (Utilities|DoOnceThenDo false
        (:DoOnce
          (Variables|Default|SetSavedPoint (+ _returnvalue (* (Math|Vector|GetForwardVector (Pawn|GetBaseAimRotation _asplayer_pawn_athena)) 1.5e+04))))
        (:ThenDo
          (Effects|Components|ParticleSystem|SetBeamTargetPoint (Utilities|Array|Get(acopy) ParticleComponents) 0 (Math|Interpolation|VInterpTo (Variables|Default|GetSavedPoint) (+ _returnvalue (* (Math|Vector|GetForwardVector (Pawn|GetBaseAimRotation _asplayer_pawn_athena)) 1.5e+04)) (stale EventTick DeltaSeconds) 3e+01) 0))))
    (:then_1
      (if (not (Pawn|IsLocallyControlled (Utilities|Casting|CastToFortPawn MyTarget)))
        (Audio|Components|Audio|FadeIn (Variables|Default|GetHightower_Tomato_Secondary_Charge_FlyBy) 0.5 1.0 0.0 "Linear" :!class /Script/Engine.AudioComponent)))))

(event Custom|Update%20Audio%20Location
  (bind _rootcomponent (Variables|Transformation|GetRootComponent))
  (bind _returnvalue (Transformation|GetForwardVector _rootcomponent))
  (bind _rootcomponent_2 (Variables|Transformation|GetRootComponent))
  (bind _returnvalue_2 (Transformation|GetWorldLocation _rootcomponent_2))
  (bind _hightower_tomato_secondary_charge_flyby (Variables|Default|GetHightower_Tomato_Secondary_Charge_FlyBy))
  ;[1 1 1] Move audio component
  (bind (_listenerposition _returnvalue_3) (Audio|GetClosestListenerLocation (Variables|Transformation|GetRootComponent) _returnvalue_2 (ScalableFloat|GetValueatLevel (Variables|Default|GetMaxRange)) true))
  (if _returnvalue_3
    (if (Audio|Components|Audio|IsPlaying (Variables|Default|GetHightower_Tomato_Secondary_Charge_FlyBy))
      (if (> (Math|Vector|DotProduct _returnvalue (- _listenerposition _returnvalue_2)) 0.0)
        (Transformation|SetWorldLocation _hightower_tomato_secondary_charge_flyby (+ _returnvalue_2 (Math|Vector|ProjectVectorontoVector (- _listenerposition _returnvalue_2) _returnvalue)) false false)
        (else
          (Audio|Components|Audio|Stop _hightower_tomato_secondary_charge_flyby))))))

(fn OnExecute
  (params (MyTarget :type Actor%20Object%20Reference) (Parameters :type Gameplay%20Cue%20Parameters%20Structure%20%28by%20ref%29))
  (return false))

(fn ConstructionScript)
