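; Looping gameplay cue for the Llama Roaster charge-up, handling muzzle effects, sound and heat-driven camera shake.
; On loop start it caches the pawn's weapon, particle and audio components, then binds to the weapon's overheat-change event.
; Each change prints debug values and starts a local-player camera shake, scaling roll amplitude by squared heat ratio; despite the handler's name, no audio parameters change.
; Removal fades the sound and unbinds.
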(event EventBeginPlay)

(event Collision|EventActorBeginOverlap
  (params (OtherActor :type Actor%20Object%20Reference)))

(event EventTick
  (params (DeltaSeconds :type Float%20%28single-precision%29)))

(event EventOnLoopingStart(EitherVFX-preferthisversion)
  (params (MyTarget :type Actor%20Object%20Reference) (Parameters :type Gameplay%20Cue%20Parameters%20Structure%20%28by%20ref%29) (ParticleComponents :type Array%20of%20FXSystem%20Component%20Object%20References) (AudioComponents :type Array%20of%20Audio%20Component%20Object%20References))
  (bind _asfort_player_pawn (Utilities|Casting|CastToFortPlayerPawn MyTarget))
  (bind _output_get (Variables|Default|SetFort%20Player%20Pawn _asfort_player_pawn))
  (bind _asb_ranged_generic (Utilities|Casting|CastToB_Ranged_Generic (Variables|/Script/FortniteGame.FortPawn|GetCurrentWeapon _output_get)))
  (Variables|Default|SetWeapon _asb_ranged_generic)
  (|CallParentFunction 0 Parameters ParticleComponents AudioComponents)
  (bind _result (ArrayMacros|SafeArrayGet ParticleComponents)
    (:%20%20
      (Variables|Default|SetMuzzleChargeFX _result)
      (bind _muzzlebuildsound (Variables|Default|GetMuzzleBuildSound)
        (:then
          (Audio|Components|Audio|FadeOut _muzzlebuildsound 0.5 0.0 "Linear" :!class /Script/Engine.AudioComponent))
        (:else
          _))
      (bind _result_2 (ArrayMacros|SafeArrayGet AudioComponents)
        (:%20%20
          (Variables|Default|SetMuzzleBuildSound _result_2)
          ;[1 1 1] Start looping timer to update the weapon overheat.
          (Overheating|BindEventtoOnOverheatValueChanged (Variables|Default|GetWeapon) (delegate Custom|UpdateAudioParamsOnOverheat)))))))

(event Custom|UpdateAudioParamsOnOverheat
  (params (CurrentValue :type Float%20%28single-precision%29) (MaxValue :type Float%20%28single-precision%29) (State :type EFortWeaponOverheatState%20Enum) (bIsInCooldown :type Boolean))
  (bind _returnvalue (/ CurrentValue MaxValue))
  (bind _fort_player_pawn (Variables|Default|GetFort%20Player%20Pawn))
  (bind _returnvalue_2 (Pawn|GetLocalViewingPlayerController _fort_player_pawn))
  ;[1 1 1] Start looping timer to update the weapon overheat.
  (Utilities|FlowControl|Sequence
    (:then_0
      (Development|PrintString 0 (Utilities|String|Append "Overheat Alpha:" (Utilities|String|ToString(Float) _returnvalue) "  Curremt: " (Utilities|String|ToString(Float) CurrentValue) "/" (Utilities|String|ToString(Float) MaxValue) "   State:" (Utilities|String|EnumtoString State) "   Is in cooldown:" (Utilities|String|ToString(Boolean) bIsInCooldown)) false false))
    (:then_1
      ;[1 1 1] UpdateCameraShake
      (Utilities|IsValid _returnvalue_2
        (:Is%20Valid
          (bind _returnvalue_3 (CameraShakes|StartCameraShake :self (Variables|/Script/Engine.PlayerController|GetPlayerCameraManager _returnvalue_2) :ShakeClass "/MotherGameplay/Items/LlamaRoaster/CameraShake_LlamaRoaster_Heatup.CameraShake_LlamaRoaster_Heatup_C" :Scale 1.0 :PlaySpace "CameraLocal"))
          (bind _output_get (Variables|Default|SetmatineeCamShake (CameraShakes|ConvLegacyCameraShake _returnvalue_3)))
          (Variables|/Script/EngineCameras.LegacyCameraShake|SetRotOscillation (Utilities|Struct|MakeROscillator "(Amplitude=0.000000,Frequency=0.000000,InitialOffset=EOO_OffsetRandom,Waveform=SineWave)" "(Amplitude=0.000000,Frequency=0.000000,InitialOffset=EOO_OffsetRandom,Waveform=SineWave)" (Utilities|Struct|MakeFOscillator (Math|Float|Lerp (Variables|Default|GetMinCamShakeAmp) (Variables|Default|GetMaxCamShakeAmp) (* _returnvalue _returnvalue)) 7e+01 "PerlinNoise")) _output_get))) ; Making sure this is valid not to spam logs when AI uses the weapon.
    )))

(event EventOnRemoval(EitherVFX-preferthisversion)
  (params (MyTarget :type Actor%20Object%20Reference) (Parameters :type Gameplay%20Cue%20Parameters%20Structure%20%28by%20ref%29) (ParticleComponents :type Array%20of%20FXSystem%20Component%20Object%20References) (AudioComponents :type Array%20of%20Audio%20Component%20Object%20References) (BurstCameraShakeInstance :type Camera%20Shake%20Base%20Object%20Reference) (BurstDecalInstance :type Decal%20Actor%20Object%20Reference))
  (|CallParentFunction MyTarget Parameters ParticleComponents AudioComponents BurstCameraShakeInstance BurstDecalInstance)
  (Development|PrintString 0 "Removed" true true "(R=1.000000,G=0.000000,B=0.127922,A=1.000000)" 5.0)
  (Audio|Components|Audio|FadeOut (Variables|Default|GetMuzzleBuildSound) 1.0 0.0 "Linear" :!class /Script/Engine.AudioComponent)
  ;[1 1 1] Stop looping timer
  (Overheating|UnbindEventfromOnOverheatValueChanged (Variables|Default|GetWeapon) (delegate Custom|UpdateAudioParamsOnOverheat)))

(fn ConstructionScript)
