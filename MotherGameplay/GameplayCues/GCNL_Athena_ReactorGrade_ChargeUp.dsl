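; Looping gameplay cue for the Reactor Grade sniper's charge-up, driving beam FX and a line-shaped audio source.
; On loop start it enables tick, passes data-table charge times and a local-player flag to the FX, and enables the audio line only for non-local pawns.
; The aim update multi-traces along weapon range, choosing the end point from landscape hits or a building pass-through limit.
; End-point updates move both FX and audio line.
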(event EventOnEndPointUpdated
  (params (FoundPoint :type Vector%20%28by%20ref%29 :default "0, 0, 0"))
  ;[1 1 1] Interpolate the fx rotation to match control rotation
  (CallFunction|SetEndPoint FoundPoint))

(event EventOnLoopingStart(EitherVFX-preferthisversion)
  (params (MyTarget :type Actor%20Object%20Reference) (Parameters :type Gameplay%20Cue%20Parameters%20Structure%20%28by%20ref%29) (ParticleComponents :type Array%20of%20FXSystem%20Component%20Object%20References) (AudioComponents :type Array%20of%20Audio%20Component%20Object%20References))
  ;[1 1 1] Interpolate the fx rotation to match control rotation
  (CallFunction|SetMinMaxChargeTimes)
  (Actor|Tick|SetActorTickEnabled true)
  (bind _asfort_pawn (Utilities|Casting|CastToFortPawn MyTarget))
  (Variables|ReactorGradeGameplayCue_ChargeUp|SetPawn _asfort_pawn)
  (bind _output_get (Variables|Default|SetLoopingFX ParticleComponents))
  (Effects|Components|ParticleSystem|SetBoolParameter _output_get "LocalPlayer" (Pawn|IsLocallyControlled _asfort_pawn))
  (Effects|Components|ParticleSystem|SetFloatParameter (Variables|Default|GetLoopingFX) "WarmUpTime" (Variables|Default|GetMax%20Charge%20Time))
  (Effects|Components|ParticleSystem|SetFloatParameter (Variables|Default|GetLoopingFX) "MinWarmUpTime" (Variables|Default|GetMin%20Charge%20Time))
  (if (not (Pawn|IsLocallyControlled _asfort_pawn))
    (Shape|Enable (Variables|Default|GetFortAudioShapeLine))))

(event EventOnRemoval(EitherVFX-preferthisversion)
  (params (MyTarget :type Actor%20Object%20Reference) (Parameters :type Gameplay%20Cue%20Parameters%20Structure%20%28by%20ref%29) (ParticleComponents :type Array%20of%20FXSystem%20Component%20Object%20References) (AudioComponents :type Array%20of%20Audio%20Component%20Object%20References) (BurstCameraShakeInstance :type Camera%20Shake%20Base%20Object%20Reference) (BurstDecalInstance :type Decal%20Actor%20Object%20Reference))
  ;[1 1 1] Handle the dissipation audio by setting AutoDestroy delay to 4s and gradually fading out audio after removing the GCNL
  (Actor|Tick|SetActorTickEnabled false)
  (Shape|Disable (Variables|Default|GetFortAudioShapeLine)))

(fn OnExecute
  (params (MyTarget :type Actor%20Object%20Reference) (Parameters :type Gameplay%20Cue%20Parameters%20Structure%20%28by%20ref%29))
  (return false))

(fn SetEndPoint
  (params (EndPoint :type Vector))
  (bind _fortaudioshapeline (Variables|Default|GetFortAudioShapeLine))
  (Effects|Components|ParticleSystem|SetVectorParameter (Variables|Default|GetLoopingFX) "EndPoint" EndPoint)
  (Control|SetStartPoint _fortaudioshapeline (Transformation|GetActorLocation))
  (Control|SetEndPoint _fortaudioshapeline EndPoint))

(fn SetMinMaxChargeTimes
  (bind _outrow (Fortnite|GetRangedWeaponStatsRow (Variables|Default|GetData%20Table%20Row%20Handle)))
  (Variables|Default|SetMin%20Charge%20Time (.MinChargeTime (Utilities|Struct|BreakFortRangedWeaponStats _outrow)))
  (Variables|Default|SetMax%20Charge%20Time (.MaxChargeTimeUntilDischarge (Utilities|Struct|BreakFortRangedWeaponStats _outrow))))

(fn UpdateAimEndPointBP
  (params (DeltaTime :type Float%20%28double-precision%29))
  (bind _actors_to_ignore (LocalVariables|GetActors%20To%20Ignore))
  (bind _pawn_0 (Variables|Default|GetPawn_0)
    (:then
      (LocalVariables|SetIsLocal (Pawn|IsLocallyControlled _pawn_0))
      (Utilities|Array|AppendArray _actors_to_ignore (Actor|GetAttachedActors _pawn_0 true false))
      (Utilities|Array|Add _actors_to_ignore _pawn_0)
      (bind _asb_rifle_sniper_reactor_grade_athena (Utilities|Casting|CastToB_Rifle_Sniper_ReactorGrade_Athena (Variables|/Script/FortniteGame.FortPawn|GetCurrentWeapon _pawn_0)))
      (bind _output_get (LocalVariables|SetEndPoint (+ (Class|BRifleSniperReactorGradeAthena|GetDamageStartandDirection _asb_rifle_sniper_reactor_grade_athena) (* (.Direction (Class|BRifleSniperReactorGradeAthena|GetDamageStartandDirection _asb_rifle_sniper_reactor_grade_athena)) (FortWeapon|GetRange _asb_rifle_sniper_reactor_grade_athena)))))
      (bind _outhits (Collision|MultiLineTraceByChannel 0 (Class|BRifleSniperReactorGradeAthena|GetDamageStartandDirection _asb_rifle_sniper_reactor_grade_athena) _output_get "TraceTypeQuery13" false (LocalVariables|GetActors%20To%20Ignore)))
      (bind (_array_element _array_index) (Utilities|Array|ForEachLoopwithBreak _outhits)
        (:LoopBody
          (if (Collision|BreakHitResult _array_element)
            (if (.bSuccess (Utilities|Casting|CastToLandscapeHeightfieldCollisionComponent (.HitComponent (Collision|BreakHitResult _array_element))))
              (LocalVariables|SetEndPoint (.Location (Collision|BreakHitResult _array_element))))
            (else
              (Math|Integer|CompareInt _array_index (Math|Float|Truncate (ScalableFloat|GetValueatLevel (Variables|Default|GetRow_TraceThroughBuildingsLimit) 1.0))
                (:<
                  (LocalVariables|SetEndPoint (.Location (Collision|BreakHitResult _array_element))))))))
        (:Completed
          (return (LocalVariables|GetEndPoint)))))
    (:else)))

(fn ConstructionScript)
