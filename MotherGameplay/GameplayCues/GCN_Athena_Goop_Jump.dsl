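; Gameplay cue notify for an Athena goop jump, which signals the jump to the target pawn's animation.
; On execute it casts the target to FortPawn, creates an anim input event named aie.athena.goop.jump and triggers it on that pawn.
; A failed cast triggers nothing, and the return value comes from the parent implementation.
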(fn OnExecute
  (params (MyTarget :type Actor%20Object%20Reference) (Parameters :type Gameplay%20Cue%20Parameters%20Structure%20%28by%20ref%29))
  (bind _asfort_pawn (Utilities|Casting|CastToFortPawn MyTarget)
    (:then
      (bind _returnvalue (Class|FortAnimInputEvent|CreateAnimInputEvent "aie.athena.goop.jump"))
      (Animation|TriggerAnimInputEvent _asfort_pawn _returnvalue))
    (:CastFailed
      _))
  (return (|CallParentFunction MyTarget Parameters)))
