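; Burst gameplay cue that plays the damaged sound for the AvacadoEater bird item.
; On burst, it spawns the damaged sound cue attached to the target actor's root component.
; It then takes the effect causer's instigator pawn and, if valid, sets the sound's IsLocal parameter from whether that pawn is locally controlled, so the audio can differ for the local player.
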(fn OnBurst(EitherVFX-preferthisversion)
  (params (MyTarget :type Actor%20Object%20Reference) (Parameters :type Gameplay%20Cue%20Parameters%20Structure%20%28by%20ref%29) (ParticleComponents :type Array%20of%20FXSystem%20Component%20Object%20References) (AudioComponents :type Array%20of%20Audio%20Component%20Object%20References) (BurstCameraShakeInstance :type Camera%20Shake%20Base%20Object%20Reference) (BurstDecalInstance :type Decal%20Actor%20Object%20Reference))
  (bind (_normalizedmagnitude _rawmagnitude _effectcontext _matchedtagname _originaltag _aggregatedsourcetags _aggregatedtargettags _location _normal _instigator _effectcauser) (Ability|GameplayCue|BreakGameplayCueParameters Parameters))
  (bind _instigator_2 (Variables|/Script/Engine.Actor|GetInstigator _effectcauser))
  (bind _returnvalue (Audio|SpawnSoundAttached "/MotherGameplay/Sounds/AvacadoEater/Bird/AvacadoEaterBird_Damaged_Cue.AvacadoEaterBird_Damaged_Cue" (Variables|/Script/Engine.Actor|GetRootComponent MyTarget)))
  (Utilities|IsValid _instigator_2
    (:Is%20Valid
      (Audio|Parameter|SetBooleanParameter _returnvalue "IsLocal" (Pawn|IsLocallyControlled _instigator_2) :!class /Script/AudioExtensions.AudioParameterControllerInterface))))
