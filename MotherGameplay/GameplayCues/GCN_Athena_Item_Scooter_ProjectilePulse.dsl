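; Burst gameplay cue for the Scooter item's projectile pulse that sets up a timed sweep effect around the target actor.
; On burst, it writes sweep duration, stencil duration and radius into the shared MPC_Interrogate material parameter collection, reused from Creative's Interrogate assets.
; It also stamps the current game time and target location as activation time and world origin, leaving any visuals to materials reading that collection.
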(fn OnBurst(EitherVFX-preferthisversion)
  (params (MyTarget :type Actor%20Object%20Reference) (Parameters :type Gameplay%20Cue%20Parameters%20Structure%20%28by%20ref%29) (ParticleComponents :type Array%20of%20FXSystem%20Component%20Object%20References) (AudioComponents :type Array%20of%20Audio%20Component%20Object%20References) (BurstCameraShakeInstance :type Camera%20Shake%20Base%20Object%20Reference) (BurstDecalInstance :type Decal%20Actor%20Object%20Reference))
  (Rendering|Material|SetScalarParameterValue MyTarget "/FNGameplayCues/Creative/Items/Gameplay/Interrogate/MPC_Interrogate.MPC_Interrogate" "Duration" (Variables|Default|GetSweepDuration) :!class /Script/Engine.KismetMaterialLibrary)
  (Rendering|Material|SetScalarParameterValue MyTarget "/FNGameplayCues/Creative/Items/Gameplay/Interrogate/MPC_Interrogate.MPC_Interrogate" "StencilDuration" (Variables|Default|GetSweepDuration) :!class /Script/Engine.KismetMaterialLibrary)
  (Rendering|Material|SetScalarParameterValue MyTarget "/FNGameplayCues/Creative/Items/Gameplay/Interrogate/MPC_Interrogate.MPC_Interrogate" "Radius" (Variables|Default|GetSweepRadius) :!class /Script/Engine.KismetMaterialLibrary)
  (Rendering|Material|SetScalarParameterValue 0 "/FNGameplayCues/Creative/Items/Gameplay/Interrogate/MPC_Interrogate.MPC_Interrogate" "ActivatedTimestamp" (Utilities|Time|GetGameTimeinSeconds MyTarget) :!class /Script/Engine.KismetMaterialLibrary)
  (Rendering|Material|SetVectorParameterValue 0 "/FNGameplayCues/Creative/Items/Gameplay/Interrogate/MPC_Interrogate.MPC_Interrogate" "OriginPositionWS" (Math|Conversions|ToLinearColor(Vector) (Transformation|GetActorLocation MyTarget)) :!class /Script/Engine.KismetMaterialLibrary)
  (return))
