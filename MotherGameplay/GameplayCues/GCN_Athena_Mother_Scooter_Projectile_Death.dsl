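; Burst gameplay cue for the death of the Mother Scooter projectile, reusing Boulder parameter names and force feedback assets.
; Treats raw magnitude as speed, feeding "BoulderVelocity" on the first particle component: a float in the EitherVFX handler, a truncated int in the Niagara one.
; Above a magnitude threshold, EitherVFX also plays large/small impact force feedback at the cue location, with magnitude-scaled, capped distances.
; The Cascade handler only prints a debug string, and OnExecute defers to the parent.
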(fn OnBurst(Cascade)
  (params (MyTarget :type Actor%20Object%20Reference) (Parameters :type Gameplay%20Cue%20Parameters%20Structure%20%28by%20ref%29) (ParticleComponents :type Array%20of%20Cascade%20Particle%20System%20Component%20Object%20References) (AudioComponents :type Array%20of%20Audio%20Component%20Object%20References) (BurstCameraShakeInstance :type Camera%20Shake%20Base%20Object%20Reference) (BurstDecalInstance :type Decal%20Actor%20Object%20Reference))
  (Development|PrintString 0 "Cascade")
  (return))

(fn OnBurst(EitherVFX-preferthisversion)
  (params (MyTarget :type Actor%20Object%20Reference) (Parameters :type Gameplay%20Cue%20Parameters%20Structure%20%28by%20ref%29) (ParticleComponents :type Array%20of%20FXSystem%20Component%20Object%20References) (AudioComponents :type Array%20of%20Audio%20Component%20Object%20References) (BurstCameraShakeInstance :type Camera%20Shake%20Base%20Object%20Reference) (BurstDecalInstance :type Decal%20Actor%20Object%20Reference))
  (bind (_normalizedmagnitude _rawmagnitude _effectcontext _matchedtagname _originaltag _aggregatedsourcetags _aggregatedtargettags _location) (Ability|GameplayCue|BreakGameplayCueParameters Parameters))
  (Effects|Components|ParticleSystem|SetFloatParameter (Utilities|Array|Get(acopy) ParticleComponents) "BoulderVelocity" _rawmagnitude)
  (Utilities|FlowControl|Sequence
    (:then_0
      (if (> _rawmagnitude 2e+03)
        (Fortnite|PlayLocalForceFeedbackatLocationMulti MyTarget "/MotherGameplay/GameplayCues/Boulder/FF_Boulder_LargeImpact.FF_Boulder_LargeImpact" "/MotherGameplay/GameplayCues/Boulder/FF_Boulder_SmallImpact.FF_Boulder_SmallImpact" _location (Math|Float|Clamp(Float) (* _rawmagnitude 0.05) 1.0 1.7e+03) (Math|Float|Min(Float) (* _rawmagnitude 0.25) 6e+03))))
    (:then_1
      (return))))

(fn OnBurst(Niagara)
  (params (MyTarget :type Actor%20Object%20Reference) (Parameters :type Gameplay%20Cue%20Parameters%20Structure%20%28by%20ref%29) (ParticleComponents :type Array%20of%20Niagara%20Particle%20System%20Component%20Object%20References) (AudioComponents :type Array%20of%20Audio%20Component%20Object%20References) (BurstCameraShakeInstance :type Camera%20Shake%20Base%20Object%20Reference) (BurstDecalInstance :type Decal%20Actor%20Object%20Reference))
  (bind (_normalizedmagnitude _rawmagnitude) (Ability|GameplayCue|BreakGameplayCueParameters Parameters))
  (Niagara|SetNiagaraVariable(Int32) (Utilities|Array|Get(acopy) ParticleComponents) "BoulderVelocity" (Math|Float|Truncate _rawmagnitude))
  (return))

(fn OnExecute
  (params (MyTarget :type Actor%20Object%20Reference) (Parameters :type Gameplay%20Cue%20Parameters%20Structure%20%28by%20ref%29))
  (return (|CallParentFunction MyTarget Parameters)))
