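; Gameplay cue notify for firing the Reactor Grade sniper rifle, laying a line-shaped audio source along the shot path for pawns that are not locally controlled.
; On burst, it multi-line-traces from the weapon along base aim to weapon range, moving the end point to qualifying hits, including landscape.
; It then enables the FortAudioShapeLine from the cue actor to that end point and disables it on a short timer.
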(event EventBeginPlay)

(event Collision|EventActorBeginOverlap
  (params (OtherActor :type Actor%20Object%20Reference)))

(event EventTick
  (params (DeltaSeconds :type Float%20%28single-precision%29)))

(fn DisableAudio
  (Shape|Disable (Variables|Default|GetFortAudioShapeLine)))

(fn OnBurst(EitherVFX-preferthisversion)
  (params (MyTarget :type Actor%20Object%20Reference) (Parameters :type Gameplay%20Cue%20Parameters%20Structure%20%28by%20ref%29) (ParticleComponents :type Array%20of%20FXSystem%20Component%20Object%20References) (AudioComponents :type Array%20of%20Audio%20Component%20Object%20References) (BurstCameraShakeInstance :type Camera%20Shake%20Base%20Object%20Reference) (BurstDecalInstance :type Decal%20Actor%20Object%20Reference))
  (bind _fortaudioshapeline (Variables|Default|GetFortAudioShapeLine))
  (bind _asfort_pawn (Utilities|Casting|CastToFortPawn MyTarget))
  (if (not (Pawn|IsLocallyControlled _asfort_pawn))
    (bind _asb_rifle_sniper_reactor_grade_athena (Utilities|Casting|CastToB_Rifle_Sniper_ReactorGrade_Athena (Variables|/Script/FortniteGame.FortPawn|GetCurrentWeapon _asfort_pawn)))
    (LocalVariables|SetStartPoint (Transformation|GetActorLocation _asb_rifle_sniper_reactor_grade_athena))
    (LocalVariables|SetEndPoint (+ (Transformation|GetActorLocation _asb_rifle_sniper_reactor_grade_athena) (* (Math|Vector|GetForwardVector (Pawn|GetBaseAimRotation _asfort_pawn)) (FortWeapon|GetRange _asb_rifle_sniper_reactor_grade_athena))))
    (bind _outhits (Collision|MultiLineTraceByChannel 0 (Transformation|GetActorLocation _asb_rifle_sniper_reactor_grade_athena) (+ (Transformation|GetActorLocation _asb_rifle_sniper_reactor_grade_athena) (* (Math|Vector|GetForwardVector (Pawn|GetBaseAimRotation _asfort_pawn)) (FortWeapon|GetRange _asb_rifle_sniper_reactor_grade_athena))) "TraceTypeQuery13"))
    (bind _array_element (Utilities|Array|ForEachLoopwithBreak _outhits)
      (:LoopBody
        (if (or (Collision|BreakHitResult _array_element) (.bSuccess (Utilities|Casting|CastToLandscapeHeightfieldCollisionComponent (.HitComponent (Collision|BreakHitResult _array_element)))))
          (LocalVariables|SetEndPoint (.Location (Collision|BreakHitResult _array_element)))))
      (:Completed
        (Shape|Enable (Variables|Default|GetFortAudioShapeLine))
        (Control|SetStartPoint _fortaudioshapeline (Transformation|GetActorLocation))
        (Control|SetEndPoint _fortaudioshapeline (LocalVariables|GetEndPoint))
        (Utilities|Time|SetTimerbyEvent (EventDispatchers|CreateEvent :!event DisableAudio) 4.0)
        (return)))))

(fn ConstructionScript)
