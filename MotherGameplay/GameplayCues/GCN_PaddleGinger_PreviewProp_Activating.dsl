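; Gameplay cue notify for the PaddleGinger preview prop activating, in the Mother gameplay set.
; On execute it casts the target to a Fort player pawn, then casts that pawn's current weapon to the PaddleGinger Mother weapon Blueprint.
; The graph ends at that cast, so it triggers no effect or call on the weapon.
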(fn OnExecute
  (params (MyTarget :type Actor%20Object%20Reference) (Parameters :type Gameplay%20Cue%20Parameters%20Structure%20%28by%20ref%29))
  (bind _asfort_player_pawn (Utilities|Casting|CastToFortPlayerPawn MyTarget))
  (bind _asb_weapon_paddle_ginger_mother (Utilities|Casting|CastToB_Weapon_PaddleGinger_Mother (Variables|/Script/FortniteGame.FortPawn|GetCurrentWeapon _asfort_player_pawn))))
