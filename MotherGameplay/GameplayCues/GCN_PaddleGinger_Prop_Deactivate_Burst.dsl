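; Burst gameplay cue for deactivating the PaddleGinger prop, with a particle variation for when the prop is a doghouse.
; OnBurst casts the target to a player pawn and its current weapon to the PaddleGinger weapon.
; It compares that weapon's preview static mesh to the suburban doghouse mesh and writes the result to the first particle component's Is_DogHouse parameter.
; OnExecute repeats the casts without using their results and returns the parent implementation's result.
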(fn OnBurst(EitherVFX-preferthisversion)
  (params (MyTarget :type Actor%20Object%20Reference) (Parameters :type Gameplay%20Cue%20Parameters%20Structure%20%28by%20ref%29) (ParticleComponents :type Array%20of%20FXSystem%20Component%20Object%20References) (AudioComponents :type Array%20of%20Audio%20Component%20Object%20References) (BurstCameraShakeInstance :type Camera%20Shake%20Base%20Object%20Reference) (BurstDecalInstance :type Decal%20Actor%20Object%20Reference))
  (bind _asfort_player_pawn (Utilities|Casting|CastToFortPlayerPawn MyTarget))
  (bind _asb_weapon_paddle_ginger_mother (Utilities|Casting|CastToB_Weapon_PaddleGinger_Mother (Variables|/Script/FortniteGame.FortPawn|GetCurrentWeapon _asfort_player_pawn)))
  (Effects|Components|ParticleSystem|SetBoolParameter (Utilities|Array|Get(acopy) ParticleComponents) "Is_DogHouse" (Utilities|Equal(Object) (Variables|/Script/Engine.StaticMeshComponent|GetStaticMesh (Variables|/MotherGameplay/Items/PaddleGinger/B_Weapon_PaddleGinger_Mother.B_Weapon_PaddleGinger_Mother_C|GetStaticMeshPreview _asb_weapon_paddle_ginger_mother)) "/Game/Environments/Props/Suburban/Meshes/SM_Doghouse.SM_Doghouse"))
  (return))

(fn OnExecute
  (params (MyTarget :type Actor%20Object%20Reference) (Parameters :type Gameplay%20Cue%20Parameters%20Structure%20%28by%20ref%29))
  (bind _asfort_player_pawn (Utilities|Casting|CastToFortPlayerPawn MyTarget))
  (bind _asb_weapon_paddle_ginger_mother (Utilities|Casting|CastToB_Weapon_PaddleGinger_Mother (Variables|/Script/FortniteGame.FortPawn|GetCurrentWeapon _asfort_player_pawn)))
  (return (|CallParentFunction MyTarget Parameters)))
