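; Gameplay cue notify for PaddleGinger, named for a prop deactivating and changing back, that shows a sound indicator when applied.
; On application it casts the target to a Fort player pawn and gets the current weapon's sound indicator component.
; It shows the indicator at the weapon's location using the PaddleGinger transform-end sound cue.
; OnBurst only calls the parent, and the graph has no cast-failure or missing-weapon handling.
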(event EventOnApplication(EitherVFX-preferthisone)
  (params (MyTarget :type Actor%20Object%20Reference) (Parameters :type Gameplay%20Cue%20Parameters%20Structure%20%28by%20ref%29) (ParticleComponents :type Array%20of%20FXSystem%20Component%20Object%20References) (AudioComponents :type Array%20of%20Audio%20Component%20Object%20References) (BurstCameraShakeInstance :type Camera%20Shake%20Base%20Object%20Reference) (BurstDecalInstance :type Decal%20Actor%20Object%20Reference))
  (bind _asfort_player_pawn (Utilities|Casting|CastToFortPlayerPawn MyTarget))
  (SoundIndicator|ShowSoundIndicator (Variables|/Script/FortniteGame.FortWeapon|GetSoundIndicatorComponent (Variables|/Script/FortniteGame.FortPawn|GetCurrentWeapon _asfort_player_pawn)) (Transformation|GetActorLocation (Variables|/Script/FortniteGame.FortPawn|GetCurrentWeapon _asfort_player_pawn)) "/MotherGameplay/Sounds/PaddleGinger/PaddleGinger_Transform_End_Cue.PaddleGinger_Transform_End_Cue" 1.0 0))

(fn OnBurst
  (params (MyTarget :type Actor%20Object%20Reference) (Parameters :type Gameplay%20Cue%20Parameters%20Structure%20%28by%20ref%29))
  (bind _returnvalue (|CallParentFunction MyTarget Parameters))
  (return _returnvalue))

(fn ConstructionScript)
