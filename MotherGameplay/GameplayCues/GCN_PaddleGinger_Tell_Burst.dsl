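; Burst gameplay cue for the PaddleGinger weapon's glitch tell, which adds a sound indicator for the tell sound.
; On execute, it casts the target to a player pawn and its current weapon to the Mother PaddleGinger weapon.
; It triggers that weapon's sound indicator component at the weapon's location with the glitch tell sound cue, then calls the parent implementation.
; The burst event carries no logic of its own.
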(fn OnBurst(EitherVFX-preferthisversion)
  (params (MyTarget :type Actor%20Object%20Reference) (Parameters :type Gameplay%20Cue%20Parameters%20Structure%20%28by%20ref%29) (ParticleComponents :type Array%20of%20FXSystem%20Component%20Object%20References) (AudioComponents :type Array%20of%20Audio%20Component%20Object%20References) (BurstCameraShakeInstance :type Camera%20Shake%20Base%20Object%20Reference) (BurstDecalInstance :type Decal%20Actor%20Object%20Reference)))

(fn OnExecute
  (params (MyTarget :type Actor%20Object%20Reference) (Parameters :type Gameplay%20Cue%20Parameters%20Structure%20%28by%20ref%29))
  (bind _asfort_player_pawn (Utilities|Casting|CastToFortPlayerPawn MyTarget))
  (bind _asb_weapon_paddle_ginger_mother (Utilities|Casting|CastToB_Weapon_PaddleGinger_Mother (Variables|/Script/FortniteGame.FortPawn|GetCurrentWeapon _asfort_player_pawn)))
  (SoundIndicator|ShowSoundIndicator (Variables|/Script/FortniteGame.FortWeapon|GetSoundIndicatorComponent _asb_weapon_paddle_ginger_mother) (Transformation|GetActorLocation _asb_weapon_paddle_ginger_mother) "/MotherGameplay/Sounds/PaddleGinger/PaddleGinger_GlitchTell_Cue.PaddleGinger_GlitchTell_Cue" 1.0 0)
  (return (|CallParentFunction MyTarget Parameters)))
